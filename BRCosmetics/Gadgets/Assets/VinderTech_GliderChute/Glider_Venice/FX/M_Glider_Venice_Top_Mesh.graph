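# Translucent, unlit, two-sided Niagara particle material for a mesh effect on the top of the Venice glider.
# A panning noise normal map's red channel distorts the opacity texture's UVs.
# Steep V-gradient masks fill one end solid and cut off the other, then particle alpha, view-angle fade and depth fade shape opacity.
# Particle color feeds a shared FX emissive adjustment function.
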Material /BRCosmetics/Gadgets/Assets/VinderTech_GliderChute/Glider_Venice/FX/M_Glider_Venice_Top_Mesh
graph {
 properties {
  allowTranslucentCustomDepthWrites=true
  bUsedWithNiagaraMeshParticles=true
  bUsedWithNiagaraSprites=true
  blendMode=BLEND_Translucent
  pixelDepthOffsetMode=PDOM_Legacy
  shadingModel=MSM_Unlit
  twoSided=true
  usedShadingModels=Unlit
 }
 nodes [
  n1 = ParticleColor()
  n2 = TextureCoordinate()
  n3 = AppendVector(
   A=Multiply(A=ComponentMask(None=@n2) {r=true}, B=ScalarParameter() {defaultValue=1, parameterName=Distortion U})
   B=Multiply(A=ComponentMask(None=@n2) {g=true}, B=ScalarParameter() {defaultValue=0.5, parameterName=Distortion V})
  )
  n4 = Constant2Vector() {g=0.2}
  n5 = TextureSampleParameter2D(UVs=Panner(Coordinate=Add(A=@n3, B=@n4)) {speedY=1.5}) {
   parameterName=Distortion Texture
   samplerType=SAMPLERTYPE_Normal
   texture=/Game/Characters/Player/Female/Medium/Bodies/F_MED_Bistro_Astronaut/FX/T_GaussianNoise_Normal
  }
  n6 = ScalarParameter() {defaultValue=0.3, parameterName=Distortion Strength}
  n7 = TextureSampleParameter2D(UVs=Add(A=TextureCoordinate() {uTiling=1.5}, B=Multiply(A=@n5.R, B=@n6) {constB=0.3})) {parameterName=Opacity Texture, texture=&/T_Glider_Venice_Top_Mesh}
  n8 = ComponentMask(None=TextureCoordinate()) {g=true}
  n9 = Saturate(None=Multiply(A=Subtract(A=@n8) {constB=0.7}) {constB=10})
  n10 = Subtract(A=ComponentMask(None=TextureCoordinate()) {g=true}) {constB=0.15}
  n11 = Multiply(A=Saturate(None=Add(A=@n7.R, B=@n9)), B=Saturate(None=Multiply(A=@n10) {constB=10}))
  n12 = ScalarParameter() {defaultValue=1, parameterName="FadeSensitivity (optional) (S)"}
  DepthFade(
   Opacity=Multiply(
    A=Multiply(A=@n1.A, B=@n11)
    B=MaterialFunctionCall("FadeSensitivity (optional)"=@n12) {materialFunction=&FadeBasedOnViewAngle}.Result
   )
   FadeDistance=ScalarParameter() {defaultValue=100, parameterName=FadeDistance}
  )
  MaterialFunctionCall(Emissive=@n1.RGB) {materialFunction=&MF_FX_EmissiveAdjustments}
 ]
}
